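# Level Sequence for authoring the LivelySugar Syrup skin's menu animation on a spawned FortPlayerMannequin, about 394 frames at 30 fps.
# A skeletal animation track sits alongside Control Rig tracks for the body (MR_Biped_v01, three sections) and face (Fortnite_3L_OnFace_CtrlRig).
# The face rig layers three blinks with brow motion and eye darts over a held asymmetric smile.
# Keyed face curves return to their starting values at the end, allowing a seamless loop.
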LevelSequence /BRCosmetics/Animation/Game/MainPlayer/Menu/BR/CustomSkins/_S41/LivelySugar_Syrup_F/Lively_Sugar_Syrup_anim/LivelySugar_Syrup_anim
bindings [FortPlayerMannequin {
  id=#1
  possessed_class=FortPlayerMannequin
  tracks [
   {type=MovieSceneSpawnTrack, sections [{type=MovieSceneSpawnSection, channels [None {type=MovieSceneScriptingBoolChannel, default=true}]}]}
   {
    type=MovieScene3DTransformTrack
    sections [{
      type=MovieScene3DTransformSection
      channels [
       Location.X {type=MovieSceneScriptingDoubleChannel, default=0}
       Location.Y {type=MovieSceneScriptingDoubleChannel, default=0}
       Location.Z {type=MovieSceneScriptingDoubleChannel, default=0}
       Rotation.X {type=MovieSceneScriptingDoubleChannel, default=0}
       Rotation.Y {type=MovieSceneScriptingDoubleChannel, default=0}
       Rotation.Z {type=MovieSceneScriptingDoubleChannel, default=0}
       Scale.X {type=MovieSceneScriptingDoubleChannel, default=1}
       Scale.Y {type=MovieSceneScriptingDoubleChannel, default=1}
       Scale.Z {type=MovieSceneScriptingDoubleChannel, default=1}
      ]
     }]
   }
   {type=MovieSceneSkeletalAnimationTrack, sections [{type=MovieSceneSkeletalAnimationSection, end_frame=394, start_frame=0}]}
   {
    type=MovieSceneControlRigParameterTrack
    display_name=MR_Biped_v01
    sections [{type=MovieSceneControlRigParameterSection}, {type=MovieSceneControlRigParameterSection}, {type=MovieSceneControlRigParameterSection}]
   }
   {
    type=MovieSceneControlRigParameterTrack
    display_name=Fortnite_3L_OnFace_CtrlRig
    sections [{
      type=MovieSceneControlRigParameterSection
      channels [
       CTRL_C_eyesAim.Location.X {type=MovieSceneScriptingFloatChannel, default=0}
       CTRL_C_eyesAim.Location.Y {type=MovieSceneScriptingFloatChannel, default=0}
       CTRL_C_eyesAim.Location.Z {type=MovieSceneScriptingFloatChannel, default=0}
       CTRL_C_eyesAim.Rotation.X {type=MovieSceneScriptingFloatChannel, default=0}
       CTRL_C_eyesAim.Rotation.Y {type=MovieSceneScriptingFloatChannel, default=0}
       CTRL_C_eyesAim.Rotation.Z {type=MovieSceneScriptingFloatChannel, default=0}
       CTRL_C_eyesAim.Scale.X {type=MovieSceneScriptingFloatChannel, default=1}
       CTRL_C_eyesAim.Scale.Y {type=MovieSceneScriptingFloatChannel, default=1}
       CTRL_C_eyesAim.Scale.Z {type=MovieSceneScriptingFloatChannel, default=1}
       CTRL_L_eyeAim.Location.X {type=MovieSceneScriptingFloatChannel, default=0}
       CTRL_L_eyeAim.Location.Y {type=MovieSceneScriptingFloatChannel, default=0}
       CTRL_L_eyeAim.Location.Z {type=MovieSceneScriptingFloatChannel, default=0}
       CTRL_L_eyeAim.Rotation.X {type=MovieSceneScriptingFloatChannel, default=0}
       CTRL_L_eyeAim.Rotation.Y {type=MovieSceneScriptingFloatChannel, default=0}
       CTRL_L_eyeAim.Rotation.Z {type=MovieSceneScriptingFloatChannel, default=0}
       CTRL_L_eyeAim.Scale.X {type=MovieSceneScriptingFloatChannel, default=1}
       CTRL_L_eyeAim.Scale.Y {type=MovieSceneScriptingFloatChannel, default=1}
       CTRL_L_eyeAim.Scale.Z {type=MovieSceneScriptingFloatChannel, default=1}
       CTRL_R_eyeAim.Location.X {type=MovieSceneScriptingFloatChannel, default=0}
       CTRL_R_eyeAim.Location.Y {type=MovieSceneScriptingFloatChannel, default=0}
       CTRL_R_eyeAim.Location.Z {type=MovieSceneScriptingFloatChannel, default=0}
       CTRL_R_eyeAim.Rotation.X {type=MovieSceneScriptingFloatChannel, default=0}
       CTRL_R_eyeAim.Rotation.Y {type=MovieSceneScriptingFloatChannel, default=0}
       CTRL_R_eyeAim.Rotation.Z {type=MovieSceneScriptingFloatChannel, default=0}
       CTRL_R_eyeAim.Scale.X {type=MovieSceneScriptingFloatChannel, default=1}
       CTRL_R_eyeAim.Scale.Y {type=MovieSceneScriptingFloatChannel, default=1}
       CTRL_R_eyeAim.Scale.Z {type=MovieSceneScriptingFloatChannel, default=1}
       CTRL_convergenceSwitch {type=MovieSceneScriptingFloatChannel, default=0}
       follow_head {type=MovieSceneScriptingFloatChannel, default=0}
       Face_Ctrl_Master.Location.X {type=MovieSceneScriptingFloatChannel, default=0}
       Face_Ctrl_Master.Location.Y {type=MovieSceneScriptingFloatChannel, default=0}
       Face_Ctrl_Master.Location.Z {type=MovieSceneScriptingFloatChannel, default=0}
       Face_Ctrl_Master.Rotation.X {type=MovieSceneScriptingFloatChannel, default=0}
       Face_Ctrl_Master.Rotation.Y {type=MovieSceneScriptingFloatChannel, default=0}
       Face_Ctrl_Master.Rotation.Z {type=MovieSceneScriptingFloatChannel, default=0}
       Face_Ctrl_Master.Scale.X {type=MovieSceneScriptingFloatChannel, default=1}
       Face_Ctrl_Master.Scale.Y {type=MovieSceneScriptingFloatChannel, default=1}
       Face_Ctrl_Master.Scale.Z {type=MovieSceneScriptingFloatChannel, default=1}
       gui_follow_head {type=MovieSceneScriptingFloatChannel, default=1}
       CTRL_L_eye_blink {
        type=MovieSceneScriptingFloatChannel
        default=0
        keys [[0,0],[101,0],[103,0.8662],[105,1.0638],[109,0.2077],[113,0.0757],[180,0],[181,0.1764],[183,0.8662],[184,1.0638],[188,0.2077],[192,0.0757],[340,0],[341,0.1256],[342,0.8662],[344,1.0638],[348,0.2077],[352,0.0757],[393,0]]
       }
       CTRL_L_eye_eyelidU.X {type=MovieSceneScriptingFloatChannel, default=0}
       CTRL_L_eye_eyelidU.Y {type=MovieSceneScriptingFloatChannel, default=0}
       CTRL_L_eye_eyelidD.X {type=MovieSceneScriptingFloatChannel, default=0}
       CTRL_L_eye_eyelidD.Y {type=MovieSceneScriptingFloatChannel, default=0}
       CTRL_L_brow.X {type=MovieSceneScriptingFloatChannel, default=0, thinned_from=19, keys [[5,0],[180,0],[184,0.0539],[189,0.0516],[194,0],[393,0]]}
       CTRL_L_brow.Y {
        type=MovieSceneScriptingFloatChannel
        default=0
        keys [[5,0],[22,0.0807],[30,0.1058],[101,0.0807],[106,-0.1445],[112,0.1021],[116,0.1494],[122,0.1665],[180,0.0807],[184,-0.2097],[189,0.0562],[194,0.1308],[223,0.1116],[231,0.0687],[285,0.0202],[339,0],[345,-0.2905],[350,-0.0661],[355,0.0084],[365,0.0388],[394,0]]
       }
       CTRL_L_brow_inner {type=MovieSceneScriptingFloatChannel, default=0}
       CTRL_L_brow_outer {type=MovieSceneScriptingFloatChannel, default=0}
       CTRL_C_nose.X {type=MovieSceneScriptingFloatChannel, default=0}
       CTRL_C_nose.Y {type=MovieSceneScriptingFloatChannel, default=0}
       CTRL_L_nose.X {type=MovieSceneScriptingFloatChannel, default=0}
       CTRL_L_nose.Y {type=MovieSceneScriptingFloatChannel, default=0}
       noseWrinkleUpper {type=MovieSceneScriptingFloatChannel, default=0}
       CTRL_R_nose.X {type=MovieSceneScriptingFloatChannel, default=0}
       CTRL_R_nose.Y {type=MovieSceneScriptingFloatChannel, default=0}
       noseWrinkleUpper_2 {type=MovieSceneScriptingFloatChannel, default=0}
       CTRL_L_eye_squintInner {type=MovieSceneScriptingFloatChannel, default=0}
       CTRL_L_eye_cheekRaise {type=MovieSceneScriptingFloatChannel, default=0}
       CTRL_C_mouth.X {type=MovieSceneScriptingFloatChannel, default=0}
       CTRL_C_mouth.Y {type=MovieSceneScriptingFloatChannel, default=0}
       CTRL_L_mouth_lipU.X {type=MovieSceneScriptingFloatChannel, default=0}
       CTRL_L_mouth_lipU.Y {type=MovieSceneScriptingFloatChannel, default=0}
       mouthPress {type=MovieSceneScriptingFloatChannel, default=0}
       lipsTogether {type=MovieSceneScriptingFloatChannel, default=0}
       CTRL_L_mouth_lipD.X {type=MovieSceneScriptingFloatChannel, default=0}
       CTRL_L_mouth_lipD.Y {type=MovieSceneScriptingFloatChannel, default=0}
       mouthPress_2 {type=MovieSceneScriptingFloatChannel, default=0}
       lipsTogether_2 {type=MovieSceneScriptingFloatChannel, default=0}
       CTRL_L_mouth_smileDepress.X {type=MovieSceneScriptingFloatChannel, default=0, const [0,1]}
       CTRL_L_mouth_smileDepress.Y {type=MovieSceneScriptingFloatChannel, default=0.1201, const [0.1201,1]}
       CTRL_L_mouth_corner.X {type=MovieSceneScriptingFloatChannel, default=-0.0541, const [-0.0541,1]}
       CTRL_L_mouth_corner.Y {type=MovieSceneScriptingFloatChannel, default=0.3568, const [0.3568,1]}
       lipsBlow {type=MovieSceneScriptingFloatChannel, default=0, const [0,1]}
       mouthDimple {type=MovieSceneScriptingFloatChannel, default=0, const [0,1]}
       mouthStretch {type=MovieSceneScriptingFloatChannel, default=0, const [0,1]}
       mouthPurseU {type=MovieSceneScriptingFloatChannel, default=0, const [0,1]}
       mouthPurseD {type=MovieSceneScriptingFloatChannel, default=0, const [0,1]}
       mouthFunnelU {type=MovieSceneScriptingFloatChannel, default=0, const [0,1]}
       mouthFunnelD {type=MovieSceneScriptingFloatChannel, default=0, const [0,1]}
       mouthCornerSharpnessU {type=MovieSceneScriptingFloatChannel, default=0, const [0,1]}
       mouthCornerSharpnessD {type=MovieSceneScriptingFloatChannel, default=0, const [0,1]}
       CTRL_R_mouth_lipU.X {type=MovieSceneScriptingFloatChannel, default=0}
       CTRL_R_mouth_lipU.Y {type=MovieSceneScriptingFloatChannel, default=0}
       mouthPress_3 {type=MovieSceneScriptingFloatChannel, default=0}
       lipsTogether_3 {type=MovieSceneScriptingFloatChannel, default=0}
       CTRL_R_mouth_lipD.X {type=MovieSceneScriptingFloatChannel, default=0}
       CTRL_R_mouth_lipD.Y {type=MovieSceneScriptingFloatChannel, default=0}
       mouthPress_2_2 {type=MovieSceneScriptingFloatChannel, default=0}
       lipsTogether_2_2 {type=MovieSceneScriptingFloatChannel, default=0}
       CTRL_R_mouth_smileDepress.X {type=MovieSceneScriptingFloatChannel, default=0.0191, const [0.0191,1]}
       CTRL_R_mouth_smileDepress.Y {type=MovieSceneScriptingFloatChannel, default=0.2266, const [0.2266,1]}
       CTRL_R_mouth_corner.X {type=MovieSceneScriptingFloatChannel, default=0.5106, const [0.5106,1]}
       CTRL_R_mouth_corner.Y {type=MovieSceneScriptingFloatChannel, default=0.616, const [0.616,1]}
       lipsBlow_2 {type=MovieSceneScriptingFloatChannel, default=0, const [0,1]}
       mouthDimple_2 {type=MovieSceneScriptingFloatChannel, default=0, const [0,1]}
       mouthStretch_2 {type=MovieSceneScriptingFloatChannel, default=0, const [0,1]}
       mouthPurseU_2 {type=MovieSceneScriptingFloatChannel, default=0, const [0,1]}
       mouthPurseD_2 {type=MovieSceneScriptingFloatChannel, default=0, const [0,1]}
       mouthFunnelU_2 {type=MovieSceneScriptingFloatChannel, default=0, const [0,1]}
       mouthFunnelD_2 {type=MovieSceneScriptingFloatChannel, default=0, const [0,1]}
       mouthCornerSharpnessU_2 {type=MovieSceneScriptingFloatChannel, default=0, const [0,1]}
       mouthCornerSharpnessD_2 {type=MovieSceneScriptingFloatChannel, default=0, const [0,1]}
       CTRL_L_mouth_suckBlow {type=MovieSceneScriptingFloatChannel, default=0}
       CTRL_C_jaw.Location.X {type=MovieSceneScriptingFloatChannel, default=0}
       CTRL_C_jaw.Location.Y {type=MovieSceneScriptingFloatChannel, default=0}
       CTRL_C_jaw.Location.Z {type=MovieSceneScriptingFloatChannel, default=0}
       jawOpenExtreme {type=MovieSceneScriptingFloatChannel, default=0}
       chinRaiseUL {type=MovieSceneScriptingFloatChannel, default=0}
       chinRaiseUR {type=MovieSceneScriptingFloatChannel, default=0}
       chinRaiseDL {type=MovieSceneScriptingFloatChannel, default=0}
       chinRaiseDR {type=MovieSceneScriptingFloatChannel, default=0}
       CTRL_R_brow.X {type=MovieSceneScriptingFloatChannel, default=0, thinned_from=19, keys [[5,0],[180,0],[184,0.0037],[189,0],[393,0]]}
       CTRL_R_brow.Y {
        type=MovieSceneScriptingFloatChannel
        default=0
        keys [[5,0],[22,0.1977],[30,0.2211],[100,0.1977],[105,-0.0276],[111,0.2179],[116,0.2706],[121,0.2811],[180,0.1977],[184,0.0385],[189,0.1954],[194,0.2382],[223,0.2094],[231,0.1125],[285,0.0318],[339,0],[345,-0.2905],[350,-0.0661],[355,0.0084],[365,0.0388],[394,0]]
       }
       CTRL_R_brow_inner {type=MovieSceneScriptingFloatChannel, default=0}
       CTRL_R_brow_outer {type=MovieSceneScriptingFloatChannel, default=0}
       CTRL_R_eye_blink {
        type=MovieSceneScriptingFloatChannel
        default=0
        keys [[0,0],[100,0],[103,0.8662],[104,1.0638],[109,0.2077],[113,0.0757],[181,0],[182,0.1256],[183,0.8662],[185,1.0638],[189,0.2077],[193,0.0757],[340,0],[341,0.1256],[342,0.8662],[344,1.0638],[348,0.2077],[352,0.0757],[393,0]]
       }
       CTRL_R_eye_eyelidU.X {type=MovieSceneScriptingFloatChannel, default=0}
       CTRL_R_eye_eyelidU.Y {type=MovieSceneScriptingFloatChannel, default=0}
       CTRL_R_eye_eyelidD.X {type=MovieSceneScriptingFloatChannel, default=0}
       CTRL_R_eye_eyelidD.Y {type=MovieSceneScriptingFloatChannel, default=0}
       CTRL_R_eye_squintInner {type=MovieSceneScriptingFloatChannel, default=0}
       CTRL_R_eye_cheekRaise {type=MovieSceneScriptingFloatChannel, default=0}
       CTRL_R_mouth_suckBlow {type=MovieSceneScriptingFloatChannel, default=0}
       CTRL_C_eye.X {
        type=MovieSceneScriptingFloatChannel
        default=-0.2917
        keys [[0,-0.3383],[20,-0.2025],[35,-0.1937],[77,-0.2033],[89,-0.2286],[103,-0.2286],[105,-0.7543],[184,-0.7543],[186,0.1534],[205,0.096],[220,-0.2595],[256,-0.149],[308,-0.3352],[340,-0.3314],[393,-0.3383]]
       }
       CTRL_C_eye.Y {
        type=MovieSceneScriptingFloatChannel
        default=0.1051
        keys [[0,0.1537],[20,0.5131],[35,0.5161],[53,0.2698],[68,0.1944],[89,0.1709],[103,0.1709],[105,-0.0423],[184,-0.0423],[186,0.2252],[205,-0.1663],[220,-0.5228],[256,-0.6529],[308,-0.5602],[340,-0.1025],[393,0.1537]]
       }
       CTRL_L_eye.X {
        type=MovieSceneScriptingFloatChannel
        default=0
        thinned_from=13
        keys [[0,0],[32,0],[33,-0.0577],[91,-0.0577],[92,-0.0191],[93,-0.0093],[265,-0.0093],[266,0.0553],[267,0.0744],[372,0.0744],[373,0.0176],[374,0]]
       }
       CTRL_L_eye.Y {type=MovieSceneScriptingFloatChannel, default=0, thinned_from=14, keys [[0,0],[160,0],[161,-0.076],[372,-0.076],[373,-0.018],[374,0]]}
       CTRL_R_eye.X {
        type=MovieSceneScriptingFloatChannel
        default=0
        thinned_from=13
        keys [[0,0],[32,0],[33,-0.0577],[91,-0.0577],[92,-0.0191],[93,-0.0093],[265,-0.0093],[266,0.0553],[267,0.0744],[372,0.0744],[373,0.0176],[374,0]]
       }
       CTRL_R_eye.Y {type=MovieSceneScriptingFloatChannel, default=0, thinned_from=14, keys [[0,0],[160,0],[161,-0.076],[372,-0.076],[373,-0.018],[374,0]]}
       eyesConvergence {type=MovieSceneScriptingFloatChannel, default=0, const [0,5]}
       eyelidsReduce {type=MovieSceneScriptingFloatChannel, default=0, const [0,5]}
       lookAtSwitch {type=MovieSceneScriptingFloatChannel, default=0}
       CTRL_C_tongue.Location.X {type=MovieSceneScriptingFloatChannel, default=0}
       CTRL_C_tongue.Location.Y {type=MovieSceneScriptingFloatChannel, default=0}
       CTRL_C_tongue.Location.Z {type=MovieSceneScriptingFloatChannel, default=0}
       wideNarrow {type=MovieSceneScriptingFloatChannel, default=0}
       rollUpDown {type=MovieSceneScriptingFloatChannel, default=0}
       rollLeftRight {type=MovieSceneScriptingFloatChannel, default=0}
       CTRL_C_teethU.Location.X {type=MovieSceneScriptingFloatChannel, default=0}
       CTRL_C_teethU.Location.Y {type=MovieSceneScriptingFloatChannel, default=0}
       CTRL_C_teethU.Location.Z {type=MovieSceneScriptingFloatChannel, default=0}
       rotateFwdBack {type=MovieSceneScriptingFloatChannel, default=0}
       CTRL_C_teethD.Location.X {type=MovieSceneScriptingFloatChannel, default=0}
       CTRL_C_teethD.Location.Y {type=MovieSceneScriptingFloatChannel, default=0}
       CTRL_C_teethD.Location.Z {type=MovieSceneScriptingFloatChannel, default=0}
       rotateFwdBack_2 {type=MovieSceneScriptingFloatChannel, default=0}
      ]
     }]
   }
  ]
 }]
display_rate {denominator=1, numerator=30}
tick_resolution {denominator=1, numerator=24000}
tracks []
